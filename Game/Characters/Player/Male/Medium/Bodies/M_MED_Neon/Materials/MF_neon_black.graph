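# Material function for the black Neon player body, overriding incoming material attributes with animated rainbow glow that flashes a reactive color, then fades from a Blueprint-set time.
# Pre-skinned position feeds customized UV1 for radial waves, combining noise and a curve-atlas ramp.
# Hue-shifted waves, tiled grid and fresnel build base color.
# A static switch adds masked, emissive rainbow ribbons; audio-amplitude nodes are unconnected.
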MaterialFunction /Game/Characters/Player/Male/Medium/Bodies/M_MED_Neon/Materials/MF_neon_black
graph {nodes [
  n1 = GetMaterialAttributes(None=FunctionInput() {inputType=FunctionInput_MaterialAttributes}) {attributeGetTypes [#1, #2]}
  n2 = PreSkinnedPosition()
  n3 = Multiply(A=Noise!(World Position=VertexInterpolator(VS=@n2).PS, FilterWidth) {bTurbulence=false, levels=1, noiseFunction=NOISEFUNCTION_GradientTex, outputMin=0, scale=0.2}) {constB=0.01}
  n4 = ScalarParameter() {group=Neon, parameterName=Neon_WaveOffset}
  n5 = Time()
  n6 = Multiply(A=ComponentMask(None=TextureCoordinate() {coordinateIndex=1}) {g=true, r=true}) {constB=0.02}
  n7 = ComponentMask(None=VectorParameter() {group=Neon, parameterName=Neon_WaveCenter, defaultValue=<rgba 0 2.6 0 1>}.RGB) {g=true, r=true}
  n8 = Add(A=Multiply(A=Add(A=@n4, B=@n5)) {constB=-0.2}, B=Multiply(A=Distance(A=@n6, B=@n7)) {constB=0.1})
  n9 = CurveAtlasRowParameter(CurveTime=Frac(None=Add(A=@n3, B=@n8))) {
   atlas=&AthenaCurveAtlas01
   curve=&/neon_body_curve
   defaultValue=220
   group=Neon
   parameterName=Neon_Param
  }
  n10 = TextureSampleParameter2D(UVs=Multiply(A=TextureCoordinate(), B=ScalarParameter() {defaultValue=1, group=Neon, parameterName=Neon_GridTiling})) {
   group=Neon
   parameterName=Neon_GridTexture
   samplerType=SAMPLERTYPE_Masks
   texture=/Game/Characters/Player/Male/Medium/Bodies/M_MED_Neon/Textures/neon_Grid_10x
  }
  n11 = Add(
   A=Power(Base=@n9.G)
   B=Multiply(A=@n10.RGB, B=ScalarParameter() {defaultValue=0.01, group=Neon, parameterName=Neon_GridMult}) {constB=0.01, desc=strength of grid everywhere}
  )
  n12 = ComponentMask(None=TextureCoordinate() {coordinateIndex=1}) {g=true, r=true}
  n13 = ScalarParameter() {defaultValue=0.01, group=Neon, parameterName=RainbowFrequency}
  n14 = VectorParameter() {group=Neon, parameterName=Neon_WaveCenter, defaultValue=<rgba 0 2.6 0 1>}
  n15 = Panner(Coordinate=Distance(A=Multiply(A=@n12, B=@n13) {constB=0.01}, B=ComponentMask(None=@n14.RGB) {g=true, r=true})) {bFractionalPart=true, speedX=-0.2}
  n16 = Constant3Vector() {constant=<rgba 1 0 0 0>}
  n17 = Subtract(A=Time(), B=ScalarParameter() {defaultValue=-10, group=Neon, parameterName=Neon_GameTimeFromBP})
  n18 = ScalarParameter() {defaultValue=1, group=Neon, parameterName=Neon_ReactiveLength}
  n19 = Multiply(A=@n18) {constB=0.33}
  n20 = Saturate(None=Divide(A=Subtract(A=@n17, B=@n19) {desc="how long the reaction will stay at 0 (100%)"}, B=Multiply(A=@n18) {constB=0.66}) {desc=the length of time the reaction takes to fade})
  n21 = LinearInterpolate(
   A=VectorParameter() {group=Neon, parameterName=Neon_ReactiveColor, defaultValue=<rgba 1 0 0 1>}.RGB
   B=Multiply(
    A=MaterialFunctionCall(Hue Shift Percentage=@n15, Texture=@n16) {materialFunction=&HueShift}.Result
    B=ScalarParameter() {defaultValue=1, group=Neon, parameterName=RainbowMultiplier}
   ) {constB=2}
   Alpha=@n20
  )
  n22 = ScalarParameter() {defaultValue=2, group=Neon, parameterName=FresnelExponentReaction}
  n23 = ScalarParameter() {defaultValue=2.5, group=Neon, parameterName=Neon_FresnelExponent}
  n24 = MaterialFunctionCall!(
   Base=Fresnel(
    ExponentIn=LinearInterpolate(A=@n22, B=@n23, Alpha=@n20)
    BaseReflectFractionIn=ScalarParameter() {defaultValue=0.01, group=Neon, parameterName=Neon_BaseReflectFractionIn}
   ) {baseReflectFraction=0, exponent=2.5}
   Add=Multiply(A=@n10.RGB, B=ScalarParameter() {defaultValue=0.2, group=Neon, parameterName=Neon_GridFresnelMult}) {constB=0.1, desc=strength of gird in fresnel}
  ) {materialFunction=/Engine/Functions/Engine_MaterialFunctions02/Math/MultiplyAdd}
  n25 = Add(
   A=VectorParameter() {group=Neon, parameterName=neon_baseColor, defaultValue=<rgba 0.01 0.01 0.01 0>}.RGB
   B=Add(A=Multiply(A=@n21, B=@n11) {desc=coloured grid + wave}, B=Multiply(A=@n21, B=@n24.Result) {desc=colored fresnel})
  )
  n26 = ComponentMask(None=TextureCoordinate() {coordinateIndex=1}) {g=true, r=true}
  n27 = ComponentMask(None=VectorParameter() {group=NeonTextureRibbons, parameterName=Neon_RibbonTextureScale, defaultValue=<rgba 2.5 1 0 1>}.RGB) {g=true, r=true}
  n28 = ComponentMask(None=VectorParameter() {group=NeonTextureRibbons, parameterName=Neon_RibbonCenter, defaultValue=<rgba 0 150 0 1>}.RGB) {g=true, r=true}
  n29 = Multiply(
   A=Distance(A=Multiply(A=@n26, B=@n27), B=@n28)
   B=ScalarParameter() {defaultValue=0.0065, group=NeonTextureRibbons, parameterName=Neon_RibbonFrequency}
  ) {constB=0.01}
  n30 = Multiply(A=Time(), B=ScalarParameter() {defaultValue=0.5, group=NeonTextureRibbons, parameterName=Neon_AudioPanSpeed}) {constB=0.5}
  n31 = Time()
  n32 = ScalarParameter() {defaultValue=-10, group=Neon, parameterName=Neon_GameTimeFromBP}
  n33 = ScalarParameter() {defaultValue=1, group=Neon, parameterName=Neon_ReactiveLength}
  n34 = Subtract(A=Max(A=Subtract(A=@n31, B=@n32)) {constB=0}, B=Multiply(A=@n33) {constB=0.33}) {desc="how long the reaction will stay at 0 (100%)"}
  n35 = Multiply(A=@n33) {constB=0.66}
  n36 = LinearInterpolate(
   A=VectorParameter() {group=Neon, parameterName=Neon_ReactiveColor, defaultValue=<rgba 1 0 0 1>}.RGB
   B=MaterialFunctionCall(
    Hue Shift Percentage=Panner(Coordinate=@n29, Time=@n30) {bFractionalPart=true, speedX=-0.5}
    Texture=Constant3Vector() {constant=<rgba 1 0 0 0>}
   ) {materialFunction=&HueShift}.Result
   Alpha=Saturate(None=Divide(A=@n34, B=@n35) {desc=the length of time the reaction takes to fade})
  ) {desc="a : reractive,\r\nb : non reactive"}
  n37 = TextureSampleParameter2D() {
   group=Neon
   parameterName=M
   samplerType=SAMPLERTYPE_Masks
   texture=&neon_Grid_10x
  }
  n38 = VectorParameter() {group=Neon, parameterName=Neon_MSR, defaultValue=<rgba 0 0.02 0.29 0>}
  n39 = Constant3Vector() {constant=<rgba 0 0 1 0>}
  n40 = VectorParameter() {group=Neon, parameterName=Opacity, defaultValue=<rgba 1 1 1 0>}
  n41 = PreSkinnedPosition()
  n42 = StaticSwitchParameter(True=ComponentMask(None=@n41) {g=true, r=true}, False=ComponentMask(None=@n41) {b=true, r=true}) {desc=use XY axis for coords used by circular ramp. If false use XZ, group=Neon, parameterName=XY axis}
  n43 = SetMaterialAttributes!(
   MaterialAttributes=@n1.MaterialAttributes
   BaseColor=LinearInterpolate(A=@n25, B=@n36, Alpha=@n37.R)
   Metallic=@n38.R
   Specular=@n38.R
   Roughness=@n38.R
   CustomizedUV1=@n42
   Normal=@n39
   Opacity=@n40.RGB
   EmissiveColor=Multiply(A=@n36, B=@n37.R)
  ) {attributeSetTypes [
    #3
    #4
    #5
    #6
    #7
    #8
    #9
    #1
   ]}
  n44 = VectorParameter() {group=Neon, parameterName=Neon_MSR, defaultValue=<rgba 0 0.02 0.29 0>}
  n45 = SetMaterialAttributes!(
   MaterialAttributes=@n1.MaterialAttributes
   BaseColor=@n25
   Metallic=@n44.R
   Specular=Add(A=Multiply(A=@n44.G) {constB=0.5}, B=Multiply(A=@n44.G, B=@n10.RGB)) {constB=0.05}
   Roughness=@n44.B
   CustomizedUV1=@n42
   Normal=@n39
   Opacity=@n40.RGB
  ) {attributeSetTypes [
    #3
    #4
    #5
    #6
    #7
    #8
    #9
   ]}
  FunctionOutput(None=StaticSwitchParameter(True=@n43, False=@n45) {group=Neon, parameterName=Neon_TexturedRibbons})
  n46 = CollectionParameter() {collection=&FortniteMaterialParameters, parameterName=EventAudio_Amplitude}
  n47 = Multiply(A=Time()) {constB=0.1}
  If(A=@n46, "A > B"=Add(A=@n47) {constB=0.2}, "A == B"=@n47, "A < B"=@n47) {constB=0.75}
  Reroute()
  SphereMask(A=@n26, B=@n28) {attenuationRadius=50}
 ]}
