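# Material function for Neon player body materials that widens spline-like strips toward the camera-facing side axis and fakes a rounded surface.
# The cross product of the camera vector and tangent-basis V axis sets the offset direction, scaled by centered U and a widthBase-to-widthTip blend along V.
# U, pulled toward center by Flatten, forms a cylindrical normal rotated into that orthogonal basis.
# Outputs are world position offset, rounded normal and unrounded facing normal.
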MaterialFunction /Game/Characters/Player/Male/Medium/Bodies/M_MED_Neon/Materials/MF_splineThickenM
graph {nodes [
  n1 = TextureCoordinate()
  n2 = ComponentMask(None=@n1) {r=true}
  n3 = LinearInterpolate(A=@n2, Alpha=FunctionInput() {inputName=Flatten, inputType=FunctionInput_Scalar}) {constB=0.5}
  n4 = AppendVector(A=ConstantBiasScale(None=Frac(None=@n3)) {bias=-0.5, scale=2}, B=Constant())
  n5 = MaterialFunctionCall() {materialFunction=&TangentBasis}
  n6 = CrossProduct(A=Subtract(A=CameraPositionWS(), B=WorldPosition().XYZ), B=@n5.Y V)
  n7 = MaterialFunctionCall!(Vector1=@n6, Vector2=Multiply(A=@n5.Y V) {constB=-1}) {materialFunction=/Engine/Functions/Engine_MaterialFunctions02/Math/CreateThirdOrthogonalVector}
  FunctionOutput(None=MaterialFunctionCall(
    VectorToTransform=DeriveNormalZ(InXY=@n4)
    BasisX=Normalize(VectorInput=@n7.Vector1)
    BasisY=Normalize(VectorInput=@n7.Vector2)
    BasisZ=Normalize(VectorInput=@n7.Vector3)
   ) {materialFunction=&Transform3x3Matrix}.Transformed Vector) {outputName=normal}
  n8 = LinearInterpolate(
   A=FunctionInput() {inputName=widthBase, inputType=FunctionInput_Scalar, previewValue=<1 0 0 1>}
   B=FunctionInput() {inputName=widthTip, inputType=FunctionInput_Scalar, previewValue=<0.5 0 0 1>}
   Alpha=ComponentMask(None=@n1) {g=true}
  )
  n9 = Add(A=Constant() {r=-0.5}, B=@n2)
  FunctionOutput(None=Multiply(A=Multiply(A=@n8, B=@n9), B=Normalize(VectorInput=@n6))) {outputName=worldPositionOffset}
  FunctionOutput(None=@n7.Vector3) {outputName=origNormal}
 ]}
