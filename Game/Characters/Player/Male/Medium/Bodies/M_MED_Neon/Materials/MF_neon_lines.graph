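# Material function for the Neon player body's glowing line tubes, setting surface, emissive and world position offset.
# Spline thickening scales width by camera distance and vertex alpha, with panned-noise wobble added.
# Event audio drives glow, hue panning and high-frequency wobble; a Blueprint timestamp triggers reactive color and shape that revert to hue-shifting, blinking defaults.
# Vertex-color sphere masks select visible shapes, collapsing others to actor position.
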MaterialFunction /Game/Characters/Player/Male/Medium/Bodies/M_MED_Neon/Materials/MF_neon_lines
graph {nodes [
  n1 = CollectionParameter() {collection=&FortniteMaterialParameters, parameterName=EventAudio_Amplitude}
  n2 = CollectionParameter() {collection=&FortniteMaterialParameters, parameterName=Emote_AudioLow}
  n3 = CollectionParameter() {collection=&FortniteMaterialParameters, parameterName=Emote_AudioHigh}
  Add(A=@n1, B=Add(A=@n2, B=@n3))
  CollectionParameter() {collection=&FortniteMaterialParameters, parameterName=EventAudio_Spectrum}
  ComponentMask(None=TextureCoordinate()) {g=true}
  Constant3Vector() {constant=<rgba 1 0 0 0>}
  Constant() {r=30}
  n4 = TextureCoordinate() {uTiling=0}
  Frac(None=@n4)
  n5 = ScalarParameter() {defaultValue=1, group=Neon, parameterName=Neon_LineWidth}
  n6 = ScalarParameter() {defaultValue=1, group=NeonDepth, parameterName=Neon_RibbonCloseWidth}
  n7 = ScalarParameter() {defaultValue=1.5, group=NeonDepth, parameterName=Neon_RibbonFarWidth}
  n8 = Distance(A=WorldPosition().XYZ, B=CameraPositionWS())
  n9 = ScalarParameter() {defaultValue=200, group=NeonDepth, parameterName=Neon_DepthClose}
  n10 = ScalarParameter() {defaultValue=601, group=NeonDepth, parameterName=Neon_DepthFar}
  n11 = Saturate(None=Divide(A=Subtract(A=@n8, B=@n9), B=Subtract(A=@n10, B=@n9)))
  n12 = Multiply(A=Multiply(A=@n5, B=LinearInterpolate(A=@n6, B=@n7, Alpha=@n11)), B=VertexColor().A)
  n13 = MaterialFunctionCall!(widthTip=@n12, Flatten=ScalarParameter() {group=Neon, parameterName=Neon_FlattenLine}, widthBase=@n12) {materialFunction=&/MF_splineThickenM}
  n14 = Multiply(A=Time(), B=ScalarParameter() {defaultValue=0.5, group=Neon, parameterName=Neon_AudioPanSpeed}) {constB=0.5}
  n15 = CollectionParameter() {collection=&FortniteMaterialParameters, parameterName=EventAudio_Amplitude}
  n16 = CollectionParameter() {collection=&FortniteMaterialParameters, parameterName=EventAudio_Spectrum}
  n17 = MaterialFunctionCall(Float4=@n16) {materialFunction=&BreakOutFloat4Components}
  n18 = Add(A=Add(A=@n17.R, B=@n17.R), B=Add(A=@n17.B, B=@n17.B))
  n19 = Multiply(A=Add(A=@n15, B=Divide(A=@n18) {constB=4}), B=ScalarParameter() {defaultValue=1, group=Neon, parameterName=Neon_EventAudioMultiplier})
  n20 = Min(A=@n19) {constB=30}
  n21 = Multiply(A=@n20, B=ScalarParameter() {defaultValue=0.7, group=Neon, parameterName=Neon_AudioColorPulseStrength}) {constB=0.7}
  n22 = Panner(Coordinate=@n4, Time=If(A=@n20, "A > B"=Add(A=@n14, B=@n21) {constB=0.2}, "A == B"=@n14, "A < B"=@n14) {constB=0.1}) {bFractionalPart=true, speedY=-0.5}
  n23 = Max(A=Subtract(A=Time(), B=ScalarParameter() {group=Neon, parameterName=Neon_GameTimeFromBP})) {constB=0}
  n24 = ScalarParameter() {defaultValue=1, group=Neon, parameterName=Neon_ReactiveLength}
  n25 = Subtract(A=@n23, B=Multiply(A=@n24) {constB=0.33}) {desc="how long the reaction will stay at 0 (100%)"}
  n26 = Multiply(A=@n24) {constB=0.66}
  n27 = LinearInterpolate(
   A=VectorParameter() {group=Neon, parameterName=Neon_ReactiveColor, defaultValue=<rgba 1 0 0 1>}.RGB
   B=MaterialFunctionCall(Hue Shift Percentage=ComponentMask(None=@n22) {g=true}, Texture=Constant3Vector() {constant=<rgba 1 0 0 0>}) {materialFunction=&HueShift}.Result
   Alpha=Saturate(None=Divide(A=@n25, B=@n26) {desc=the length of time the reaction takes to fade})
  ) {desc="a : reractive,\r\nb : non reactive"}
  n28 = VertexColor()
  n29 = Multiply(A=Time()) {constB=0.2, desc=frequency of blink}
  n30 = Floor(None=Multiply(A=Frac(None=@n29)) {constB=1.04, desc=">1 = length of blink"})
  n31 = Saturate(None=Divide(A=@n23, B=Multiply(A=@n24) {constB=1.5, desc=shapes stay on longer}) {desc=the length of time the reaction takes to fade})
  n32 = SphereMask(
   A=@n28.R
   B=LinearInterpolate(A=ScalarParameter() {group=Neon, parameterName=Neon_ReactiveShape}, B=Multiply(A=@n30) {constB=0.25}, Alpha=Floor(None=@n31)) {constB=0, desc=lerp between reactive A and passive B}
   Radius=Constant() {r=0.13}
   Hardness=Constant() {r=1}
  )
  n33 = Multiply(
   A=LinearInterpolate(A=Add(A=@n32) {constB=0.02, desc=brightness of deactivated tubes}, Alpha=@n28.B) {constA=1, desc="deactivated = 0"}
   B=@n28.G
  )
  n34 = VectorParameter() {group=Neon, parameterName=Neon_MSR, defaultValue=<rgba 1 1 0.2 1>}
  n35 = Constant() {r=3}
  n36 = Multiply(
   A=MaterialFunctionCall!(Base=@n27, Add=@n20) {desc=multiplyAdd by audio, materialFunction=/Engine/Functions/Engine_MaterialFunctions02/Math/MultiplyAdd}.Result
   B=@n35
  ) {desc=multiply by TOD}
  n37 = ScalarParameter() {defaultValue=1, group=Neon, parameterName=Neon_EmissiveStrength}
  n38 = ActorPositionWS()
  n39 = WorldPosition()
  n40 = MaterialFunctionCall(Offset Amount=Multiply(A=Constant() {r=1}, B=@n33)) {materialFunction=/Engine/Functions/Engine_MaterialFunctions02/WorldPositionOffset/CameraOffset}
  n41 = TextureCoordinate() {uTiling=0}
  n42 = ScalarParameter() {defaultValue=1, group=Neon, parameterName=Neon_WobbleTileLowFreq}
  n43 = TextureSample(UVs=Panner(Coordinate=Multiply(A=@n41, B=@n42)) {bFractionalPart=true, speedX=0.1, speedY=-0.3}) {samplerType=SAMPLERTYPE_LinearColor, texture=&HighContrastNoise}
  n44 = ScalarParameter() {defaultValue=1, group=Neon, parameterName=Neon_WobbleLowMultiplier}
  n45 = Add(B=TextureCoordinate() {uTiling=0})
  n46 = ScalarParameter() {defaultValue=2.5, group=Neon, parameterName=Neon_WobbleTileHighFreq}
  n47 = VectorParameter() {group=Neon, parameterName=Neon_WobbleSpeed, defaultValue=<rgba 0.28 -0.5 0 0>}
  n48 = TextureSample(UVs=Panner(Coordinate=Multiply(A=@n45, B=@n46), Speed=ComponentMask(None=@n47.RGB) {g=true, r=true}) {bFractionalPart=true, speedX=0.25, speedY=-0.5}) {samplerType=SAMPLERTYPE_LinearColor, texture=&HighContrastNoise}
  n49 = Multiply(
   A=Multiply(A=Subtract(A=@n48.R) {constB=0.3}, B=ScalarParameter() {defaultValue=8, group=Neon, parameterName=Neon_WobbleHighMultiplier}) {constB=10}
   B=MaterialFunctionCall(Alpha=@n20, Min=Constant() {r=0.1}, Max=Constant() {r=1}) {materialFunction=&SmoothStep}.Result
  )
  n50 = ScalarParameter() {defaultValue=0.8, group=Neon, parameterName=Neon_AudioWobbleStrength}
  n51 = Multiply(
   A=Add(A=Multiply(A=@n43.R, B=@n44) {constB=1.5}, B=Multiply(A=@n49, B=@n50))
   B=ScalarParameter() {defaultValue=1, group=Neon, parameterName=Neon_WobbleStrength}
  )
  n52 = Multiply(A=MaterialFunctionCall() {materialFunction=&TangentBasis}.Z W) {constB=-1}
  n53 = Add(A=@n13.worldPositionOffset, B=Multiply(A=Multiply(A=@n51, B=@n52), B=@n28.G))
  n54 = Saturate(None=Add(A=@n32, B=@n28.B))
  n55 = SetMaterialAttributes!(
   MaterialAttributes=GetMaterialAttributes(None=FunctionInput() {inputType=FunctionInput_MaterialAttributes}) {attributeGetTypes [#1, #2]}.MaterialAttributes
   BaseColor=Multiply(A=@n27, B=@n33)
   Metallic=@n34.R
   Specular=@n34.G
   Roughness=@n34.B
   EmissiveColor=Multiply(A=@n33, B=Multiply(A=@n36, B=@n37) {desc=global Mult}) {desc=Mult by Vertex Colors}
   Normal=@n13.normal
   WorldPositionOffset=LinearInterpolate(A=Subtract(A=@n38, B=@n39.XYZ), B=Add(A=@n40.Camera Offset, B=@n53), Alpha=Round(None=@n54)) {constA=1, desc="deactivated = 0"}
  ) {attributeSetTypes [
    #3
    #4
    #5
    #6
    #1
    #7
    #2
   ]}
  FunctionOutput(None=@n55)
  n56 = ObjectRadius()
  n57 = ScalarParameter() {defaultValue=96.9905, group=Neon, parameterName=ActorRadius, sortPriority=0}
  LinearInterpolate(A=@n5, B=Multiply(A=@n5, B=Divide(A=@n56, B=@n57)), Alpha=ScalarParameter() {group=Neon, parameterName=useActorRadius, sortPriority=1})
  n58 = Constant() {r=4}
  MaterialFunctionCall(
   1=@n58
   2=@n35
   3=@n35
   4=@n58
   A=CollectionParameter() {collection=&FortniteMaterialParameters, parameterName=TimeOfDay_0-1}
  ) {materialFunction=&Lerp_Multiple_Float3}
  n59 = CollectionParameter() {collection=&FortniteMaterialParameters, parameterName=Emote_AudioLow}
  n60 = CollectionParameter() {collection=&FortniteMaterialParameters, parameterName=Emote_AudioHigh}
  Max(A=Add(A=@n59, B=@n60), B=@n19)
  Power(Base=Multiply(A=@n15) {constB=0.5})
  Power(Base=Add(A=Multiply(A=@n17.R) {constB=0.5}, B=Multiply(A=@n17.G) {constB=0.5}))
  Reroute() {desc=normal}
  VertexColor()
 ]}
